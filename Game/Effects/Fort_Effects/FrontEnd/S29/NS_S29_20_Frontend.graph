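# Niagara system for a Season 29 frontend scene, layering snowfall, drifting foreground fog and sweeping cloud puffs.
# Cloud bits spawn along a looping lerped line; a second emitter spawns from them via attribute reads, rendering faint sprites and ribbons with curl noise.
# GPU snowflakes use wind, gravity and collision-dependent forces, aging and culling, while a sparser CPU LOW variant omits collision.
# Both snow emitters feed rotating random vectors to materials for glints.
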NiagaraSystem /Game/Effects/Fort_Effects/FrontEnd/S29/NS_S29_20_Frontend
dependencies {
 usedDataInterfaces [
  &NiagaraDataInterfaceCurve
  &NiagaraDataInterfacePlatformSet
  &NiagaraDataInterfaceParticleRead
  &NiagaraDataInterfaceArrayFloat
 ]
 usedDynamicInputs [&LerpPosition]
 usedModules [&SystemState, &Timeline]
}
emitters {
 Empty {
  emitterSpawnScript [/SetVariables() {isSetParameters=true}]
  emitterUpdateScript [
   &EmitterState()
   &SpawnParticlesFromOtherEmitter(
    Attribute Reader=DataInterface() {propertyValues {EmitterBinding {bindingMode=Other, emitterName=Initial_cloud_Bits001}}}
    Spawn Rate=3
    Use Maximum Age=false
    Use Minimum Age=false
   )
  ]
  particleSpawnScript [
   &InitializeParticle(
    Color=&MultiplyLinearColor(
     Linear Color=<rgba 1 1 1 25>
     Scale Factor=&MakeFloat4FromVectorAndFloat(
      "Vector (RGB)"=<1 1 1>
      "Float (Alpha)"=&FloatFromCurve(
       FloatCurve=DataInterface() {propertyValues {Curve {keys [
           {interpMode=RCIM_Cubic, tangentMode=RCTM_User, time=0, value=0}
           {interpMode=RCIM_Cubic, tangentMode=RCTM_User, time=0.5, value=1}
           {tangentMode=RCTM_User, time=1, value=0}
          ]}}}
       CurveIndex=&Divide_Float(A=1, B=1)
      )
     )
    )
    Lifetime=7
    Ribbon Width=150
    Ribbon Width Mode=::NewEnumerator1
    Sprite Size Mode=::NewEnumerator3
    UsePositionOffset=false
    Write Material Random=true
   )
   &SampleParticlesFromOtherEmitter(
    Age Sampling:ENiagara_DirectReadSamplingMode=::NewEnumerator1
    Applied Velocity Scale=0
    Lifetime Sampling:ENiagara_DirectReadSamplingMode=::NewEnumerator1
    Particle ID Sampling=::NewEnumerator1
    Ribbon UV Distance:ENiagara_DirectReadSamplingMode=::NewEnumerator0
    Velocity Sampling=::NewEnumerator0
   )
   &Color(Color=&MultiplyLinearColor(
     Linear Color=<rgba 0.585 0.8697 0.9 0.5>
     Scale Factor=&MakeFloat4FromVectorAndFloat(
      "Vector (RGB)"=<1 1 1>
      "Float (Alpha)"=&FloatFromCurve(
       FloatCurve=DataInterface() {propertyValues {Curve {keys [
           {interpMode=RCIM_Cubic, tangentMode=RCTM_User, time=0, value=0}
           {interpMode=RCIM_Cubic, tangentMode=RCTM_User, time=0.5, value=1}
           {tangentMode=RCTM_User, time=1, value=0}
          ]}}}
       CurveIndex=&Divide_Float(A=$Output.SampleParticlesFromOtherEmitter.SampledAge, B=$Output.SampleParticlesFromOtherEmitter.SampledLifetime)
      )
     )
    ))
   &KillParticles(Kill Particles=&SetBoolByIntComparison(
     A=&ModuloInt(A=&self:ScratchDynamicInput(Niagara ID:NiagaraID=$Particles.RibbonID), B=15)
     B=0
     Comparison Type=::NewEnumerator5
    ))
   /SetVariables(Particles.Position=&self:SetZ(
     Position:NiagaraPosition=$Particles.Position
     Z:NiagaraFloat=&FloatFromCurve(
      FloatCurve=DataInterface() {propertyValues {Curve {keys [
          {interpMode=RCIM_Cubic, tangentMode=RCTM_User, time=0, value=1}
          {interpMode=RCIM_Cubic, tangentMode=RCTM_User, time=0.5, value=1}
          {tangentMode=RCTM_User, time=1, value=1}
         ]}}}
      Scale Curve=5
     )
    )) {isSetParameters=true}
   &AddVelocity(Velocity=<1 0 0>)
   &InheritVelocity(
    Inherited Velocity Amount Scale=&VectorFromFloat(Value=0.1)
    Speed Limit=false
    UseSpeedThreshold=false
    Velocity Source=$Output.SampleParticlesFromOtherEmitter.SampledVelocity
   )
  ]
  particleUpdateScript [
   &CurlNoiseForce(Curl Noise Cone Mask Axis:Vector3f=<1 0 0>, Mask Curl Noise=true, Noise Frequency=1, Pan Noise=true, Pan Noise Field=<1 1 1>)
   &LinearForce(Coordinate Space:ENiagaraCoordinateSpace=::World, Force=<100 0 0>) {enabled=false}
   &Drag(Drag=5, Use Linear Drag=true, Use Rotational Drag=true)
   &AccelerationForce(Acceleration=&MakeVector(
     X=&FloatFromCurve(
      FloatCurve=DataInterface() {propertyValues {Curve {keys [{tangentMode=RCTM_User, time=0, value=1}, {tangentMode=RCTM_User, time=1, value=0}]}}}
      Scale Curve=250
     )
     Y=0
     Z=0
    )) {enabled=false}
   &GravityForce(Gravity=&MakeVector(
     X=0
     Y=0
     Z=&FloatFromCurve(
      FloatCurve=DataInterface() {propertyValues {Curve {keys [{arriveTangent=2.8943, interpMode=RCIM_Cubic, leaveTangent=2.8943, tangentMode=RCTM_User, time=0, value=0}, <at 1 1 RCIM_Cubic>]}}}
      Scale Curve=-100
     )
    ))
   &SolveForcesAndVelocity(Clamp Velocity=false, Limit Acceleration=false)
   &ParticleState()
   &ScaleColor(
    Scale Alpha=&FloatFromCurve(FloatCurve=DataInterface() {propertyValues {Curve {keys [
         {arriveTangent=6.4804, interpMode=RCIM_Cubic, time=0, value=0}
         {interpMode=RCIM_Cubic, tangentMode=RCTM_User, time=0.5, value=1}
         {tangentMode=RCTM_User, time=1, value=0}
        ]}}})
    ScaleA=true
    ScaleRGB=true
   )
   &ScaleColor(ScaleA=true, ScaleRGB=true) {module=ScaleColor001}
  ]
  rendererClasses [&NiagaraSpriteRendererProperties, &NiagaraRibbonRendererProperties]
  renderers [
   {rendererClass=&NiagaraSpriteRendererProperties}
   {rendererClass=&NiagaraRibbonRendererProperties, rendererIndex=1}
  ]
 }
 Foreground_Fog {
  emitterUpdateScript [
   &EmitterState(Life Cycle Mode=::NewEnumerator1, Loop Duration=2, UseLoopDelay=false)
   &SpawnBurst_Instantaneous(Use Spawn Probability=false, UseLoopCountLimit=false)
   &SpawnRate(SpawnRate=4, Use Spawn Probability=false)
  ]
  particleSpawnScript [
   &InitializeParticle(
    Color Maximum=<rgba 0.585 0.8697 0.9 0.1>
    Color Minimum=<rgba 0.7105 0.9246 0.9473 0.01>
    Color Mode=::NewEnumerator2
    Lifetime Max=10
    Lifetime Min=7
    Lifetime Mode=::NewEnumerator1
    Sprite Rotation Mode=::NewEnumerator2
    Sprite Size Mode=::NewEnumerator1
    Sprite UV Mode=::NewEnumerator3
    Uniform Sprite Size Max=300
    Uniform Sprite Size Min=200
    UsePositionOffset=false
    Write Material Random=true
   )
   &ShapeLocation(
    Box Midpoint=&RandomRangeVector(Minimum=<0 0 0>, Maximum=<1 1 1>)
    Box Size=<400 1000 50>
    MaskSpawnGroup=false
    Offset=<0 250 -50>
    Offset Coordinate Space=::World
    Offset Mode=::NewEnumerator0
    Override Seed=false
    Shape Primitive=::NewEnumerator2
    Surface Only=false
    UseOwnerScale=true
   )
   &DynamicMaterialParameters(Write Parameter Index 0=false)
   &AddVelocity(
    Velocity=&RandomRangeVector(Minimum=<0 -1 0.5>, Maximum=<0 -3 -2>)
    Velocity Speed Scale=&RandomRangeFloat(Minimum=5, Maximum=30)
   )
   &CameraOffset(Camera Offset Amount=150)
  ]
  particleUpdateScript [
   &ParticleState()
   &SpriteRotationRate(Rotation Rate=&RandomRangeFloat(Minimum=15, Maximum=30))
   &ScaleSpriteSize(Scale Factor=&Vector2DFromFloat(Value=&FloatFromCurve(FloatCurve=DataInterface() {propertyValues {Curve {keys [{tangentMode=RCTM_User, time=0, value=0.75}, {tangentMode=RCTM_User, time=1, value=1}]}}})))
   &ScaleColor(
    Scale Alpha=&FloatFromCurve(FloatCurve=DataInterface() {propertyValues {Curve {keys [<at 0 0 RCIM_Cubic>,<at 0.5 1 RCIM_Cubic>,<at 1 0 RCIM_Cubic>]}}}, Scale Curve=1.25)
    ScaleA=true
    ScaleRGB=true
   )
   &SolveForcesAndVelocity(Clamp Velocity=false, Limit Acceleration=false)
  ]
  rendererClasses [&NiagaraSpriteRendererProperties]
  renderers [
   {rendererClass=&NiagaraSpriteRendererProperties}
   {rendererClass=&NiagaraSpriteRendererProperties, rendererIndex=1}
  ]
 }
 Initial_cloud_Bits001 {
  emitterUpdateScript [
   &EmitterState(
    Delay First Loop Only=true
    Life Cycle Mode=::NewEnumerator1
    Loop Delay=&RandomRangeFloat(Minimum=0, Maximum=0)
    Loop Duration=&RandomRangeFloat(Minimum=60, Maximum=300)
    Recalculate Duration Each Loop=true
    UseLoopDelay=true
   )
   &SpawnRate(
    SpawnRate=&FloatFromCurve(FloatCurve=DataInterface() {propertyValues {Curve {keys [<at 0 1>,<at 1 1>]}}}, CurveIndex=$Emitter.LoopedAge, Scale Curve=10)
    Use Spawn Probability=false
   ) {module=SpawnRate001}
  ]
  particleSpawnScript [
   &InitializeParticle(
    Color Maximum=<rgba 0.585 0.8697 0.9 0.035>
    Color Minimum=<rgba 0.7105 0.9246 0.9473 0.01>
    Color Mode=::NewEnumerator2
    Lifetime Max=3
    Lifetime Mode=::NewEnumerator1
    Mesh Scale Max=<0.75 0.7 0.7>
    Mesh Scale Min=<0.5 1 1>
    Mesh Scale Mode=::NewEnumerator2
    Position Offset=&Lerp_Vector(
     A=<-1000 400 -50>
     B=<1000 400 -100>
     Alpha=&FloatFromCurve(
      FloatCurve=DataInterface() {propertyValues {Curve {keys [{tangentMode=RCTM_User, time=0, value=0}, {tangentMode=RCTM_User, time=15, value=1}]}}}
      CurveIndex=$Emitter.LoopedAge
     )
    )
    Sprite Rotation Mode=::NewEnumerator2
    Sprite Size Mode=::NewEnumerator1
    Uniform Sprite Size Max=200
    Uniform Sprite Size Min=75
    UsePositionOffset=true
    Write Material Random=true
   )
   &AddVelocityInCone(
    Cone Angle=120
    Cone Axis Coordinate Space=::World
    Module.Use Velocity Falloff On Cone Axis=false
    Velocity Distribution Along Cone Axis=0.15
    Velocity Strength=&Multiply_Float(
     A=&RandomRangeFloat(Minimum=-250, Maximum=-400)
     B=&FloatFromCurve(
      FloatCurve=DataInterface() {propertyValues {Curve {keys [{arriveTangent=-2.7353, interpMode=RCIM_Cubic, leaveTangent=-2.7353, tangentMode=RCTM_User, time=0, value=1}, <at 1 0 RCIM_Cubic>]}}}
      CurveIndex=$Emitter.NormalizedLoopAge
     )
    )
   ) {enabled=false}
   &AddVelocity(
    Velocity=&RandomRangeVector(Minimum=<-500 0 0>, Maximum=<500 0 0>)
    Velocity Speed Scale=&Multiply_Float(A=0.5, B=$System.Timeline.Float)
   ) {enabled=false}
   &ShapeLocation(MaskSpawnGroup=false, Non Uniform Scale=<5 1 1>, Override Seed=false, Sphere Radius=50, UseOwnerScale=true)
   &CameraOffset(Camera Offset Amount=25)
   &Color(Color=&HueShiftLinearColor(Hue=0, Saturation=0.5, Value=1, Linear Color=$Particles.Color))
  ]
  particleUpdateScript [
   &ParticleState()
   &Drag(Drag=3, Use Linear Drag=true, Use Rotational Drag=false)
   &AccelerationForce(
    Acceleration=&Multiply_Vector(
     A=&RandomRangeVector(Minimum=<0 -800 0>, Maximum=<0 -1000 0>)
     B=&VectorFromFloat(Value=0.5)
    )
    Coordinate Space=::Local
   )
   &CurlNoiseForce(Noise Frequency=120, Noise Strength=500, Pan Noise=false)
   &SolveForcesAndVelocity(Clamp Velocity=false, Limit Acceleration=false)
   &SubUVAnimation()
   &DynamicMaterialParameters(
    Index 0 Param 1=&FloatFromCurve(FloatCurve=DataInterface() {propertyValues {Curve {keys [{tangentMode=RCTM_User, time=0, value=0}, {tangentMode=RCTM_User, time=1, value=1}]}}})
    Module.Param0WriteEnabled=true
    Module.Param1WriteEnabled=true
    Module.Param2WriteEnabled=true
    Module.Param3WriteEnabled=true
   )
   &SpriteRotationRate(Rotation Rate=&RandomRangeFloat(Minimum=-60, Maximum=-120))
   &ScaleColor(
    Scale Alpha=&Multiply_Float(
     A=&FloatFromCurve(FloatCurve=DataInterface() {propertyValues {Curve {keys [
          {arriveTangentWeight=0.3284, interpMode=RCIM_Cubic, leaveTangentWeight=0.3284, time=0, value=0}
          {arriveTangentWeight=0.2965, interpMode=RCIM_Cubic, leaveTangentWeight=0.2965, time=0.7, value=1}
          {arriveTangentWeight=0.3603, interpMode=RCIM_Cubic, leaveTangentWeight=0.3603, time=1, value=0}
         ]}}})
     B=&FloatFromCurve(
      FloatCurve=DataInterface() {propertyValues {Curve {keys [{tangentMode=RCTM_User, time=0.2, value=0}, {tangentMode=RCTM_User, time=1, value=1}]}}}
      CurveIndex=$Emitter.Age
     )
    )
    ScaleA=true
    ScaleRGB=true
   )
   &ScaleSpriteSize(Scale Sprite Size Mode=::NewEnumerator0, Uniform Scale Factor=$System.Timeline.Float)
   /SetVariables() {isSetParameters=true}
   /SetVariables(Particles.Dynam 1:Vector4f=<1 1 1 1>) {isSetParameters=true}
  ]
  rendererClasses [&NiagaraSpriteRendererProperties]
  renderers [
   {rendererClass=&NiagaraSpriteRendererProperties}
   {rendererClass=&NiagaraSpriteRendererProperties, rendererIndex=1}
   {rendererClass=&NiagaraSpriteRendererProperties, rendererIndex=2}
  ]
 }
 SnowflakesAir {
  simTarget=GPUComputeSim
  emitterSpawnScript [/SetVariables(
    Emitter.bool:NiagaraBool=&IsPlatformSetActive(New Platform Set=DataInterface() {propertyValues {Platforms {qualityLevelMask=-4}}})
    Emitter.float:NiagaraFloat=&MakeCustomFloatFromBool(
     Bool=&IsPlatformSetActive(New Platform Set=DataInterface() {propertyValues {Platforms {qualityLevelMask=-29}}})
     True Float=1
     False Float=&MakeCustomFloatFromBool(
      Bool=&IsPlatformSetActive(New Platform Set=DataInterface() {propertyValues {Platforms {qualityLevelMask=-28}}})
      True Float=1
      False Float=0
     )
    )
    Emitter.float001:NiagaraFloat=&MakeCustomFloatFromBool(
     Bool=&IsPlatformSetActive(New Platform Set=DataInterface() {propertyValues {Platforms {qualityLevelMask=-29}}})
     True Float=1
     False Float=0
    )
   ) {isSetParameters=true}]
  emitterUpdateScript [
   &EmitterState()
   &SpawnRate(SpawnRate=750, Use Spawn Probability=false)
  ]
  particleSpawnScript [
   /Niagara/Modules/Spawn/Initialization/InitializeParticle(
    Lifetime=&UniformRangedFloat(Minimum=3, Maximum=6)
    Module.Write Color=true
    Module.Write Lifetime=true
    Module.Write Mass=true
    Module.Write Material Random=true
    Module.Write Position=false
    Module.Write Scale=false
    Module.Write SpriteRotation=true
    Module.Write SpriteSize=true
    Sprite Rotation=&UniformRangedFloat(Minimum=0, Maximum=360)
    Sprite Size=&Vector2DFromFloat(Value=&RandomRangeFloat(Minimum=2, Maximum=10))
   ) {enabled=false}
   /Niagara/Modules/Spawn/Initialization/V2/InitializeParticle(
    Color=<rgba 2.5 2.5 2.5 1>
    Lifetime=&UniformRangedFloat(Minimum=3, Maximum=6)
    Mass Mode=::NewEnumerator1
    Position Offset=<200 0 0>
    Position Offset Coordinate Space=::World
    Sprite Rotation Mode=::NewEnumerator2
    Sprite Size=&Vector2DFromFloat(Value=&RandomRangeFloat(Minimum=2, Maximum=10))
    Sprite Size Mode=::NewEnumerator4
    UsePositionOffset=true
    Write Material Random=true
   ) {module=InitializeParticle001}
   /Niagara/Modules/Spawn/Location/SphereLocation(
    Hemisphere Z=true
    MaskSpawnGroup=false
    Module.Override Randomness=false
    Non Uniform Scale=<1 0.5 1>
    Offset=<-400 0 100>
    Sphere Radius=800
    Surface Only=false
    Surface Only Band Thickness=1
   ) {enabled=false}
   &ShapeLocation(MaskSpawnGroup=false, Offset=<-400 0 300>, Offset Mode=::NewEnumerator0, Override Seed=false, Sphere Radius=800, UseOwnerScale=true)
   &AddVelocity(Coordinate Space=::World, Velocity=&UniformRangedVector(Minimum=<100 0 -35>, Maximum=<150 0 -75>))
   &ApplyInitialForces()
  ]
  particleUpdateScript [
   &Collision(
    Advanced Aging Rate=0
    Age Colliding Particles=true
    Analytical Collision Plane Position 1:Vector3f=<0 0 0>
    Analytical Collision Plane Position 2:Vector3f=<0 0 0>
    Friction=100000
    Friction During a Bounce=1000000
    GPU Collision Type=::NewEnumerator3
    KillOnCollision=false
    Module.Control Rest State Via Surface Normal=true
    Module.EnableMaxCollisionCount=false
    Module.Randomize Collision Normal Vector=false
    Percentage of Penetration Before Rest=0.001
    Restitution=0
    Simple Friction=false
    Static Friction=100000
    Static Friction Engagement Speed=100000
   )
   &ParticleState(DeltaTime:NiagaraFloat=&Multiply_Float(
     A=$Engine.DeltaTime
     B=&MakeCustomFloatFromBool(
      Bool=&SetBoolByIntComparison(A=$Particles.Number Of Collisions, B=0, Comparison Type=::NewEnumerator0)
      True Float=3
      False Float=1
     )
    ))
   &KillParticles(Kill Particles=&Boolean_AndOperation(
     Bool1=&SetBoolByIntComparison(A=$Particles.Number Of Collisions, B=0, Comparison Type=::NewEnumerator0)
     Bool2=&SetBoolByFloatComparison(A=$Particles.Age, B=0.1, Comparison Type=::NewEnumerator6)
    ))
   &KillParticles(Kill Particles=&Boolean_AndOperation(
     Bool1=&SetBoolByFloatComparison(A=$Particles.Age, B=0.1, Comparison Type=::NewEnumerator6)
     Bool2=&Boolean_AndOperation(
      Bool1=&SetBoolByIntComparison(A=$Particles.Number Of Collisions, B=0, Comparison Type=::NewEnumerator0)
      Bool2=&SetBoolByFloatComparison(
       A=&MakeFloatFromVector(Vector=Linked() {linkedVariable {name=Particles.Position, type=NiagaraPosition}}, Channel=::NewEnumerator2)
       B=10
       Comparison Type=::NewEnumerator6
      )
     )
    )) {module=KillParticles001}
   &ScaleColor(
    Scale Alpha=&FloatFromCurve(FloatCurve=DataInterface() {propertyValues {Curve {keys [
         <at -0.0007 -0.0138 RCIM_Cubic>
         {arriveTangentWeight=0.2085, interpMode=RCIM_Cubic, leaveTangentWeight=0.1761, tangentMode=RCTM_User, time=0.5, value=1}
         <at 1.0011 0.002 RCIM_Cubic>
        ]}}})
    Scale RGB=&VectorFromFloat(Value=1)
    ScaleA=true
    ScaleRGB=true
   )
   &CurlNoiseForce(
    Noise Frequency=250
    Noise Strength=&MakeCustomFloatFromBool(
     Bool=&SetBoolByIntComparison(A=$Particles.Number Of Collisions, B=0, Comparison Type=::NewEnumerator0)
     True Float=0
     False Float=500
    )
    Pan Noise=false
   ) {module=CurlNoiseForce001}
   &VortexForce(
    UseOriginOffset=false
    Vortex Axis=<1 0 0>
    Vortex Force Amount=&MakeCustomFloatFromBool(
     Bool=&SetBoolByIntComparison(A=$Particles.Number Of Collisions, B=0, Comparison Type=::NewEnumerator0)
     True Float=0
     False Float=20
    )
   ) {enabled=false}
   &GravityForce(Gravity=&MakeCustomVectorFromBool(
     Bool=&SetBoolByIntComparison(A=$Particles.Number Of Collisions, B=0, Comparison Type=::NewEnumerator0)
     True Vector=<0 0 0>
     False Vector=<0 0 -50>
    ))
   &WindForce(
    Wind Speed=<800 0 100>
    Wind Speed Scale=&MakeCustomFloatFromBool(
     Bool=&SetBoolByIntComparison(A=$Particles.Number Of Collisions, B=0, Comparison Type=::NewEnumerator0)
     True Float=0
     False Float=&Lerp_Float(
      A=0.35
      B=0.1
      Alpha=&Waveform("Waveform [1]"=::NewEnumerator9, "[1] Frequency"=0.5, "[1] Amplitude Min/Max"=<0 1>)
     )
    )
   )
   &SolveForcesAndVelocity(Clamp Velocity=true, Limit Acceleration=false, Speed Limit=150)
   &CalculateAccurateVelocity()
   &KillParticles(Kill Particles=&Boolean_AndOperation(
     Bool1=&SetBoolByFloatComparison(A=$Particles.Age, B=0.1, Comparison Type=::NewEnumerator6)
     Bool2=&SetBoolByFloatComparison(A=&VectorLength(Vector=$Particles.Velocity), B=10, Comparison Type=::NewEnumerator6)
    )) {module=KillParticles002}
   &SpriteRotationRate(Rotation Rate=&UniformRangedFloat(Minimum=-50, Maximum=150))
   /SetVariables(Particles.RandomGlintVector:Vector3f=&RotateVectorByQuaternion(
     Vector To Rotate=&RandomVector(Vector Scale=1)
     Quaternion=&MakeQuatFromAxisAngle(
      Axis=&RandomVector(Vector Scale=1)
      AngleInDegrees=&Multiply_Float(A=$Particles.NormalizedAge, B=&RandomRangeFloat(Minimum=2000, Maximum=3000))
     )
    )) {isSetParameters=true, module=SetVariables_#1}
   &DynamicMaterialParameters(
    Index 0 Param 1=&RandomRangeFloat(Minimum=0, Maximum=1)
    Index 0 Param 2=&MakeFloatFromVector(Vector=$Particles.RandomGlintVector, Channel=::NewEnumerator0)
    Index 0 Param 3=&MakeFloatFromVector(Vector=$Particles.RandomGlintVector, Channel=::NewEnumerator1)
    Index 0 Param 4=&MakeFloatFromVector(Vector=$Particles.RandomGlintVector, Channel=::NewEnumerator2)
    Module.Param0WriteEnabled=true
    Module.Param1WriteEnabled=true
    Module.Param2WriteEnabled=true
    Module.Param3WriteEnabled=true
   )
  ]
  rendererClasses [&NiagaraSpriteRendererProperties]
  renderers [{rendererClass=&NiagaraSpriteRendererProperties}]
 }
 SnowflakesAir_LOW {
  emitterUpdateScript [
   &EmitterState()
   &SpawnRate(SpawnRate=50, Use Spawn Probability=false)
  ]
  particleSpawnScript [
   /Niagara/Modules/Spawn/Initialization/InitializeParticle(
    Lifetime=&UniformRangedFloat(Minimum=3, Maximum=6)
    Module.Write Color=true
    Module.Write Lifetime=true
    Module.Write Mass=true
    Module.Write Material Random=true
    Module.Write Position=false
    Module.Write Scale=false
    Module.Write SpriteRotation=true
    Module.Write SpriteSize=true
    Sprite Rotation=&UniformRangedFloat(Minimum=0, Maximum=360)
    Sprite Size=&Vector2DFromFloat(Value=&RandomRangeFloat(Minimum=10, Maximum=15))
   ) {enabled=false}
   /Niagara/Modules/Spawn/Initialization/V2/InitializeParticle(
    Lifetime=&UniformRangedFloat(Minimum=3, Maximum=6)
    Mass Mode=::NewEnumerator1
    Sprite Rotation Mode=::NewEnumerator2
    Sprite Size=&Vector2DFromFloat(Value=&RandomRangeFloat(Minimum=10, Maximum=15))
    Sprite Size Mode=::NewEnumerator4
    UsePositionOffset=false
    Write Material Random=true
   ) {module=InitializeParticle001}
   &ShapeLocation(MaskSpawnGroup=false, Offset=<-400 0 300>, Offset Mode=::NewEnumerator0, Override Seed=false, Sphere Radius=800, UseOwnerScale=true)
   &AddVelocity(Velocity=&UniformRangedVector(Minimum=<100 0 -35>, Maximum=<150 0 -75>))
   &ApplyInitialForces()
  ]
  particleUpdateScript [
   &ParticleState() {module=UpdateAge}
   &ScaleColor(
    Scale Alpha=&FloatFromCurve(FloatCurve=DataInterface() {propertyValues {Curve {keys [
         <at -0.0007 -0.0138 RCIM_Cubic>
         {arriveTangent=2.6721, interpMode=RCIM_Cubic, leaveTangent=2.6721, tangentMode=RCTM_User, time=0.3127, value=0.3952}
         {arriveTangent=-0.5712, interpMode=RCIM_Cubic, leaveTangent=-0.5712, tangentMode=RCTM_User, time=0.5424, value=1.0086}
         <at 1.0011 0.002 RCIM_Cubic>
        ]}}})
    ScaleA=true
    ScaleRGB=true
   )
   &CurlNoiseForce(Noise Frequency=250, Noise Strength=500, Pan Noise=false) {module=CurlNoiseForce001}
   &GravityForce(Gravity=<0 0 -50>)
   &WindForce(
    Wind Speed=<800 0 100>
    Wind Speed Scale=&Lerp_Float(
     A=0.35
     B=0.1
     Alpha=&Waveform("Waveform [1]"=::NewEnumerator9, "[1] Frequency"=0.5, "[1] Amplitude Min/Max"=<0 1>)
    )
   )
   &SolveForcesAndVelocity(Clamp Velocity=true, Limit Acceleration=false, Speed Limit=150)
   &SpriteRotationRate(Rotation Rate=&UniformRangedFloat(Minimum=-50, Maximum=150))
   /SetVariables(Particles.RandomGlintVector:Vector3f=&RotateVectorByQuaternion(
     Vector To Rotate=&RandomVector(Vector Scale=1)
     Quaternion=&MakeQuatFromAxisAngle(
      Axis=&RandomVector(Vector Scale=1)
      AngleInDegrees=&Multiply_Float(A=$Particles.NormalizedAge, B=&RandomRangeFloat(Minimum=2000, Maximum=3000))
     )
    )) {isSetParameters=true, module=SetVariables_#1}
   &DynamicMaterialParameters(
    Index 0 Param 1=&RandomRangeFloat(Minimum=0, Maximum=1)
    Index 0 Param 2=&MakeFloatFromVector(Vector=$Particles.RandomGlintVector, Channel=::NewEnumerator0)
    Index 0 Param 3=&MakeFloatFromVector(Vector=$Particles.RandomGlintVector, Channel=::NewEnumerator1)
    Index 0 Param 4=&MakeFloatFromVector(Vector=$Particles.RandomGlintVector, Channel=::NewEnumerator2)
    Module.Param0WriteEnabled=true
    Module.Param1WriteEnabled=true
    Module.Param2WriteEnabled=true
    Module.Param3WriteEnabled=true
   )
  ]
  rendererClasses [&NiagaraSpriteRendererProperties]
  renderers [{rendererClass=&NiagaraSpriteRendererProperties}]
 }
}
summary {
 systemName=NS_S29_20_Frontend
 emitters [
  {emitterName=Foreground_Fog, rendererClasses [&NiagaraSpriteRendererProperties]}
  {emitterName=SnowflakesAir, simTarget=GPUComputeSim, rendererClasses [&NiagaraSpriteRendererProperties]}
  {emitterName=SnowflakesAir_LOW, rendererClasses [&NiagaraSpriteRendererProperties]}
  {emitterName=Initial_cloud_Bits001, rendererClasses [&NiagaraSpriteRendererProperties]}
  {emitterName=Empty, rendererClasses [&NiagaraSpriteRendererProperties, &NiagaraRibbonRendererProperties]}
 ]
 userVariables [User.ParrotOffset {type=Vector3f, description="", defaultValue {struct=Vector3f, value=<0 -175 175>}}]
}
